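# UI material that draws the HUD hit marker as procedural, radially mirrored ticks, with variants for shield, crit, fatal and shield-break hits.
# UVs compensate for vertex scaling, then rotate with a random offset and radially mirror into ticks, each a skewed, anti-aliased trapezoid SDF blended with a soft shadow.
# Switch scalars lerp shape and MPC_BRGameUI colors, while Anim scalars drive size, spread, opacity and break motion, with elimination variants.
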Material /BattleRoyaleGameUI/HUD/HitTicks/Assets/M_UI_HitTicks
graph {
 properties {bAutomaticallySetUsageInEditor=false, blendMode=BLEND_Translucent, materialDomain=MD_UI, usedShadingModels=Default Lit}
 nodes [
  n1 = CollectionParameter() {collection=/BattleRoyaleGameUI/Shared/MPC_BRGameUI, parameterName=Yellow}
  n2 = ScalarParameter() {group=Switches, parameterName=Switch_Crit, sortPriority=1}
  n3 = NamedRerouteUsage()
  n4 = ComponentMask(None=Multiply(A=Add(A=@n3) {constB=1.5}) {constB=0.5}) {g=true}
  n5 = LinearInterpolate(
   A=Constant() {r=1}
   B=LinearInterpolate(
    A=CollectionParameter() {collection=&MPC_BRGameUI, parameterName=Cyan}
    B=CollectionParameter() {collection=&MPC_BRGameUI, parameterName=Blue}
    Alpha=Saturate(None=@n4)
   )
   Alpha=ScalarParameter() {group=Switches, parameterName=Switch_Shield, sortPriority=1}
  )
  LinearInterpolate(
   B=LinearInterpolate(
    A=LinearInterpolate(A=@n5, B=@n1, Alpha=@n2)
    B=CollectionParameter() {collection=&MPC_BRGameUI, parameterName=Red}
    Alpha=ScalarParameter() {group=Switches, parameterName=Switch_Fatal, sortPriority=1}
   )
   Alpha=NamedRerouteUsage()
  )
  LinearInterpolate(
   A=@n1
   B=CollectionParameter() {collection=&MPC_BRGameUI, parameterName=Orange}
   Alpha=MaterialFunctionCall("Value (0-1)"=@n4) {materialFunction=&MF_UI_EaseCurves}.EaseOutSine
  )
  MaterialFunctionCall!(Anchor, Translation, Padding, Scale=ScalarParameter() {defaultValue=2, group=Multipliers, parameterName=VertexScale, sortPriority=0}) {materialFunction=/FortUILibrary/MaterialFunctions/Basics/MF_UI_VertexTransform}
  Multiply(A=NamedRerouteUsage(), B=NamedRerouteUsage())
  n6 = MaterialFunctionCall(
   UVs=MaterialFunctionCall() {materialFunction=&GetUserInterfaceUV}.Normalized UV
   Texture Scale=NamedRerouteUsage()
  ) {materialFunction=&ScaleUVsByCenter}
  n7 = Add(A=Multiply(A=Subtract(A=@n6.UVs) {constB=0.5}, B=ScalarParameter() {defaultValue=2, group=Multipliers, parameterName=VertexScale, sortPriority=0})) {constB=0.5}
  n8 = Add(
   A=ScalarParameter() {defaultValue=0.5, group=Hit Mark Parameters, parameterName=HitMark_Rotation, sortPriority=1}
   B=LinearInterpolate(Alpha=ScalarParameter() {defaultValue=0.5, group=Hit Mark Parameters, parameterName=HitMark_RandomHitRotation, sortPriority=2}) {constA=-0.01, constB=0.01}
  )
  n9 = Rotator(
   Coordinate=MaterialFunctionCall!(
    UVs=Rotator(Coordinate=@n7, Time=@n8) {speed=6.2832}
    Sides=ScalarParameter() {defaultValue=4, group=Hit Tick Parameters, parameterName=HitTick_Count, sortPriority=0}
   ) {materialFunction=/FortUILibrary/MaterialFunctions/Basics/MF_UI_UVRadialMirror}.UVs
   Time=ScalarParameter() {group=Hit Tick Parameters, parameterName=HitTick_Rotation, sortPriority=1}
  ) {speed=6.2832}
  n10 = Divide(A=Subtract(A=ComponentMask(None=@n9) {r=true}) {constB=0.5}, B=NamedRerouteUsage())
  n11 = ComponentMask(None=@n9) {g=true}
  n12 = ScalarParameter() {defaultValue=0.25, group=Hit Tick Parameters, parameterName=HitTick_Position, sortPriority=1}
  n13 = NamedRerouteUsage()
  n14 = Divide(A=Add(A=Subtract(A=@n11), B=Add(A=@n12, B=@n13)), B=NamedRerouteUsage())
  NamedRerouteDeclaration(None=AppendVector(A=Add(A=@n10) {constB=0.5}, B=Add(A=@n14))) {name=Uvs, nodeColor=<rgba 0.25 0.2021 0.125 0.75>}
  n15 = ScalarParameter() {defaultValue=0.175, group=Hit Tick Parameters, parameterName=HitTick_Width, sortPriority=1}
  n16 = Add(A=@n15) {constB=0.075}
  n17 = ScalarParameter() {group=Switches, parameterName=AnimSwitch_Elim, sortPriority=2}
  NamedRerouteDeclaration(None=LinearInterpolate(
    A=LinearInterpolate(A=@n15, B=@n16, Alpha=@n17) {constA=0.2, constB=0.4}
    B=Add(A=@n15) {constB=-0.15}
    Alpha=ScalarParameter() {group=Animators, parameterName=Anim_TickWidth, sortPriority=2}
   ) {constA=0.2, constB=0.1, desc=TickWidth}) {name=AnimTickWidth, nodeColor=<rgba 0.125 0.101 0.0625 0.75>}
  n18 = ScalarParameter() {defaultValue=0.1, group=Hit Tick Parameters, parameterName=HitTick_Scale, sortPriority=1}
  n19 = Add(A=@n18) {constB=0.05}
  n20 = ScalarParameter() {group=Switches, parameterName=AnimSwitch_Elim, sortPriority=2}
  n21 = Add(A=@n19) {constB=0.025}
  NamedRerouteDeclaration(None=LinearInterpolate(
    A=LinearInterpolate(A=@n18, B=@n19, Alpha=@n20) {constA=0.05, constB=0.1, desc=Min}
    B=LinearInterpolate(A=@n19, B=@n21, Alpha=@n20) {constA=0.1, constB=0.2, desc=Max}
    Alpha=ScalarParameter() {group=Animators, parameterName=Anim_TickScale, sortPriority=2}
   ) {constA=0.1, constB=0.2, desc=Scale}) {name=AnimTickScale, nodeColor=<rgba 0.125 0.101 0.0625 0.75>}
  NamedRerouteDeclaration(None=LinearInterpolate(
    B=ScalarParameter() {defaultValue=1, group=Opacity, parameterName=Opacity_Master, sortPriority=0}
    Alpha=ScalarParameter() {group=Animators, parameterName=Anim_Opacity, sortPriority=0}
   ) {constB=0, desc=Opacity}) {name=AnimOpacity, nodeColor=<rgba 0.125 0.101 0.0625 0.75>}
  n22 = ScalarParameter() {defaultValue=1.1, group=Hit Mark Parameters, parameterName=HitMark_Scale, sortPriority=0}
  NamedRerouteDeclaration(None=LinearInterpolate(A=Add(A=@n22) {constB=0.25}, B=@n22, Alpha=ScalarParameter() {group=Animators, parameterName=Anim_MarkScale, sortPriority=1}) {constA=1.25, desc=HMScale}) {name=HitMarkerScale, nodeColor=<rgba 0.125 0.101 0.0625 0.75>}
  NamedRerouteDeclaration(None=LinearInterpolate(
    A=ScalarParameter() {defaultValue=0.1, group=Hit Tick Parameters, parameterName=HitTick_Scale, sortPriority=1}
    B=NamedRerouteUsage()
    Alpha=ScalarParameter() {group=Switches, parameterName=Switch_Anim, sortPriority=0}
   )) {name=HitTick_Scale, nodeColor=<rgba 0.125 0.101 0.0625 0.75>}
  NamedRerouteDeclaration(None=LinearInterpolate(
    A=ScalarParameter() {defaultValue=0.175, group=Hit Tick Parameters, parameterName=HitTick_Width, sortPriority=1}
    B=NamedRerouteUsage()
    Alpha=ScalarParameter() {group=Switches, parameterName=Switch_Anim, sortPriority=0}
   )) {name=HitTick_Width, nodeColor=<rgba 0.125 0.101 0.0625 0.75>}
  n23 = ScalarParameter() {group=Switches, parameterName=Switch_ShieldBreak, sortPriority=1}
  n24 = ScalarParameter() {group=Switches, parameterName=Switch_Crit, sortPriority=1}
  n25 = LinearInterpolate(Alpha=@n24) {constA=-0.5, constAlpha=1, constB=-0.1, desc=Symmetry Gap}
  n26 = LinearInterpolate(Alpha=@n24) {constA=2, constB=1.75}
  n27 = LinearInterpolate(Alpha=ScalarParameter() {group=Animators, parameterName=Anim_ShieldBreak, sortPriority=3}) {constB=0.625, desc=Break Animate}
  n28 = Constant() {r=2}
  n29 = ScalarParameter() {group=Switches, parameterName=Switch_ShieldBreak, sortPriority=1}
  n30 = Add(
   A=MaterialFunctionCall!(UV=NamedRerouteUsage(), Mirror Y=LinearInterpolate(Alpha=@n23) {constA=1, constB=1.5}, Swizzle UV) {materialFunction=/FortUILibrary/MaterialFunctions/Basics/MF_UI_Symmetry_2x1_New}.Symmetry
   B=LinearInterpolate(A=AppendVector(A=@n25, B=@n26), B=AppendVector(A=@n27, B=@n28), Alpha=@n29)
  )
  n31 = LinearInterpolate(Alpha=ScalarParameter() {group=Animators, parameterName=Anim_ShieldBreak, sortPriority=3}) {constB=-0.025}
  n32 = LinearInterpolate(Alpha=@n24) {constA=10, constB=7.5}
  n33 = LinearInterpolate(Alpha=@n29) {constA=35, constB=-35}
  n34 = MaterialFunctionCall!(
   UVs=MaterialFunctionCall!(UV=Rotator(Coordinate=@n30, Time=@n31) {speed=6.2832}, Skew Angle=AppendVector(A=@n32, B=@n33)) {materialFunction=/FortUILibrary/MaterialFunctions/Basics/MF_UI_Skew}.Skewed UVs
   TopWidth=Constant() {r=0.1}
   BottomWidth=LinearInterpolate(Alpha=@n23) {constA=0.75}
   Height=LinearInterpolate(Alpha=@n24) {constA=3, constB=2.75}
  ) {materialFunction=/FortUILibrary/MaterialFunctions/Shapes/SDFs/MF_SDF_IsoscelesTrapezoid}
  n35 = MaterialFunctionCall(DistanceField=@n34.SDF) {materialFunction=&MF_UI_AntiAliasing}
  n36 = MaterialFunctionCall(DistanceField=@n34.SDF, Threshold=Constant() {r=0.2}, Smoothing=Constant() {r=1}) {materialFunction=&MF_UI_AntiAliasing}
  NamedRerouteDeclaration(None=LinearInterpolate(
    A=NamedRerouteDeclaration(None=@n35.Inner) {name=HitTickMask, nodeColor=<rgba 0.125 0.101 0.0625 0.75>}
    B=NamedRerouteDeclaration(None=@n36.Inner) {name=HitTickShadow, nodeColor=<rgba 0.125 0.101 0.0625 0.75>}
   ) {constAlpha=0.4}) {name=HitTickComposite, nodeColor=<rgba 0.125 0.101 0.0625 0.75>}
  NamedRerouteDeclaration(None=@n7) {name=VertexScaledUV, nodeColor=<rgba 0.125 0.101 0.0625 0.75>}
  n37 = ScalarParameter() {defaultValue=0.125, group=Hit Tick Parameters, parameterName=HitTick_Spread, sortPriority=1}
  n38 = Add(A=@n37) {constB=-0.1}
  n39 = ScalarParameter() {group=Switches, parameterName=AnimSwitch_Elim, sortPriority=2}
  n40 = Multiply(A=@n37) {constB=-4}
  NamedRerouteDeclaration(None=LinearInterpolate(
    A=@n37
    B=LinearInterpolate(A=@n38, B=@n40, Alpha=@n39) {constA=0.3, constB=-0.25}
    Alpha=ScalarParameter() {group=Animators, parameterName=Anim_TickSpread, sortPriority=2}
   ) {constA=0.25, constB=0.3, desc=TickSpread}) {name=AnimTickSpread, nodeColor=<rgba 0.125 0.101 0.0625 0.75>}
  ScalarParameter() {defaultValue=0.1, group=Hit Tick Parameters, parameterName=HitTick_Scale, sortPriority=1}
  ScalarParameter() {defaultValue=1.1, group=Hit Mark Parameters, parameterName=HitMark_Scale, sortPriority=0}
  ScalarParameter() {defaultValue=0.125, group=Hit Tick Parameters, parameterName=HitTick_Spread, sortPriority=1}
  ScalarParameter() {group=Switches, parameterName=AnimSwitch_Elim, sortPriority=2}
  ScalarParameter() {group=Animators, parameterName=Anim_TickScale, sortPriority=2}
  ScalarParameter() {group=Animators, parameterName=Anim_MarkScale, sortPriority=1}
  ScalarParameter() {group=Animators, parameterName=Anim_TickSpread, sortPriority=2}
  ScalarParameter() {defaultValue=1, group=Opacity, parameterName=Opacity_Master, sortPriority=0}
  ScalarParameter() {group=Animators, parameterName=Anim_Opacity, sortPriority=0}
  ScalarParameter() {group=Multipliers, parameterName=Float_DamageMultipler, sortPriority=1}
  ScalarParameter() {group=Switches, parameterName=Switch_Shield, sortPriority=1}
  ScalarParameter() {group=Switches, parameterName=Switch_Anim, sortPriority=0}
  ScalarParameter() {defaultValue=0.175, group=Hit Tick Parameters, parameterName=HitTick_Width, sortPriority=1}
  ScalarParameter() {group=Switches, parameterName=AnimSwitch_Break, sortPriority=2}
  ScalarParameter() {group=Animators, parameterName=Anim_TickWidth, sortPriority=2}
  ScalarParameter() {defaultValue=2, group=Multipliers, parameterName=VertexScale, sortPriority=0}
  ScalarParameter() {defaultValue=0.5, group=Hit Mark Parameters, parameterName=HitMark_RandomHitRotation, sortPriority=2}
  ScalarParameter() {defaultValue=0.5, group=Hit Mark Parameters, parameterName=HitMark_Rotation, sortPriority=1}
  ScalarParameter() {defaultValue=4, group=Hit Tick Parameters, parameterName=HitTick_Count, sortPriority=0}
  ScalarParameter() {group=Hit Tick Parameters, parameterName=HitTick_Rotation, sortPriority=1}
  ScalarParameter() {defaultValue=0.25, group=Hit Tick Parameters, parameterName=HitTick_Position, sortPriority=1}
  ScalarParameter() {group=Switches, parameterName=Switch_Crit, sortPriority=1}
  ScalarParameter() {group=Switches, parameterName=Switch_Fatal, sortPriority=1}
  ScalarParameter() {group=Switches, parameterName=Switch_ShieldBreak, sortPriority=1}
  ScalarParameter() {group=Animators, parameterName=Anim_ShieldBreak, sortPriority=3}
 ]
}
